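# Translucent UI material for the cosmetic gate warning icon, with two presentations switched by the UseAlternativeWarningPresentation value in MPC_CosmeticGate.
# The default draws a rotated, scaled SDF triangle with backing and border colors, overlaid with a tinted, transformable exclamation texture.
# The alternative uses an antialiased MSDF information icon.
# Lerps driven by the collection value blend color and alpha between the two, with the alternative's color a constant white.
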Material /Game/Athena/UI/Frontend/CosmeticGating/Materials/M_UI_CosmeticGate_WarningIcon
graph {
 properties {
  bUsedWithStaticMesh=false
  blendMode=BLEND_Translucent
  materialDomain=MD_UI
  pixelDepthOffsetMode=PDOM_Legacy
  usedShadingModels=Default Lit
 }
 nodes [
  LinearInterpolate(A=NamedRerouteUsage(), B=Constant() {r=1}, Alpha=NamedRerouteUsage())
  LinearInterpolate(A=NamedRerouteUsage(), B=NamedRerouteUsage(), Alpha=NamedRerouteUsage())
  n1 = VectorParameter() {parameterName=TriangleOffset, defaultValue=<rgba 0.05 -0.13 0 1>}
  n2 = Add(A=TextureCoordinate(), B=AppendVector(A=@n1.R, B=@n1.R))
  n3 = ScalarParameter() {defaultValue=0.018, parameterName=RotationAngle, sliderMax=0.1}
  n4 = MaterialFunctionCall!(
   UVs=MaterialFunctionCall(
    UVs=MaterialFunctionCall(UVs=@n2, "Rotation Angle (0-1)"=@n3) {materialFunction=&CustomRotator}.Rotated Values
    Texture Scale=VectorParameter() {parameterName=TriangleScale, defaultValue=<rgba 1.08 1.2 0 1>}.RGB
   ) {materialFunction=&ScaleUVsByCenter}.UVs
   Radius=Constant() {r=1}
   Sides=Constant() {r=3}
   InnerAngle
  ) {materialFunction=/FortUILibrary/MaterialFunctions/Shapes/SDFs/MF_SDF_RegularPolygon}
  NamedRerouteDeclaration(None=MaterialFunctionCall!(
    DistanceField=@n4.SDF
    Threshold
    Thickness=ScalarParameter() {defaultValue=0.07, parameterName=OutlineThickness}
    Alignment=Constant() {r=-1}
    Smoothing
   ) {materialFunction=/FortUILibrary/MaterialFunctions/Shapes/Modifiers/MF_UI_ShapeOutline}.Outline) {name=outline, nodeColor=<rgba 1 0 0.7734 1>}
  NamedRerouteDeclaration(None=MaterialFunctionCall(DistanceField=@n4.SDF, Smoothing=Constant() {r=0.01}) {materialFunction=&MF_UI_AntiAliasing}.Inner) {name=alpha v1, nodeColor=<rgba 0.0547 1 0 1>}
  n5 = VectorParameter() {group=Color, parameterName=BackingColor, defaultValue=<rgba 0.0953 0.3564 0.6445 1>}
  n6 = VectorParameter() {group=Color, parameterName=BorderColor, defaultValue=<rgba 0.9301 1 0.0103 1>}
  n7 = NamedRerouteUsage()
  n8 = TextureCoordinate()
  n9 = ScalarParameter() {defaultValue=0.016, group=Texture, parameterName=RotateTexture}
  n10 = ScalarParameter() {defaultValue=0.056, group=Texture, parameterName=Offset_Texture_U}
  n11 = ScalarParameter() {defaultValue=-0.1154, group=Texture, parameterName=Offset_Texture_V}
  n12 = MaterialFunctionCall(
   UVs=Add(
    A=MaterialFunctionCall(UVs=@n8, "Rotation Angle (0-1)"=@n9) {materialFunction=&CustomRotator}.Rotated Values
    B=AppendVector(A=@n10, B=@n11)
   )
   Texture Scale=AppendVector(
    A=ScalarParameter() {defaultValue=0.419, group=Texture, parameterName=Scale_Texture_U}
    B=ScalarParameter() {defaultValue=0.4286, group=Texture, parameterName=Scale_Texture_V}
   )
  ) {materialFunction=&ScaleUVsByCenter}
  n13 = TextureSampleParameter2D(UVs=@n12.UVs) {
   group=Texture
   parameterName=Texture
   samplerSource=SSM_Clamp_WorldGroupSettings
   texture=/Game/UI/Foundation/Textures/Icons/Replay/T-Icon-Replay-Exclamation-L
  }
  NamedRerouteDeclaration(None=LinearInterpolate(
    A=LinearInterpolate(A=@n5.RGB, B=@n6.RGB, Alpha=@n7)
    B=VectorParameter() {group=Texture, parameterName=TextureTint, defaultValue=<rgba 0.9301 1 0.0103 1>}.RGB
    Alpha=MaterialFunctionCall!(UVs=@n12.UVs, Alpha=@n13.A) {materialFunction=/Game/UI/Foundation/Items/ClampAlphaToTransparentBorder}.Result
   )) {name=color v1, nodeColor=<rgba 1 0.4219 0 1>}
  NamedRerouteDeclaration(None=CollectionParameter() {collection=&/MPC_CosmeticGate, parameterName=UseAlternativeWarningPresentation}) {name=use v2, nodeColor=<rgba 0.1016 0 1 1>}
  n14 = TextureSample() {samplerType=SAMPLERTYPE_LinearColor, texture=/FortUILibrary/Textures/Icons/T_UI_Icon_UIKit_Information_MSDF}
  NamedRerouteDeclaration(None=MaterialFunctionCall(
    DistanceField=MaterialFunctionCall!(MSDF RGB=@n14.RGB) {materialFunction=/FortUILibrary/MaterialFunctions/Shapes/Modifiers/MF_UI_MSDFtoSDF}.SDF
    Threshold=Constant() {r=0.5}
    Smoothing=Constant() {r=0.2}
   ) {materialFunction=&MF_UI_AntiAliasing}.Outer) {name=alpha v2, nodeColor=<rgba 0.2187 1 0 1>}
 ]
}
