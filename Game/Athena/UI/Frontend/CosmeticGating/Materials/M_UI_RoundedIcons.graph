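# Translucent UI material that draws a round icon badge: a colored circle with an outline and a centered, tinted icon, under a CosmeticGating frontend folder.
# A circle function produces fill and outline masks, sized from pixel-based edge offset and outline width converted to normalized units.
# Masks blend circle and outline colors, then the icon color is layered on using a center-scaled, clamped texture read as a distance field with antialiased edges.
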Material /Game/Athena/UI/Frontend/CosmeticGating/Materials/M_UI_RoundedIcons
graph {
 properties {
  bUsedWithStaticMesh=false
  blendMode=BLEND_Translucent
  materialDomain=MD_UI
  pixelDepthOffsetMode=PDOM_Legacy
  usedShadingModels=Default Lit
 }
 nodes [
  n1 = NamedRerouteUsage()
  n2 = OneMinus(None=NamedRerouteUsage())
  n3 = TextureSample(
   UVs=MaterialFunctionCall(Texture Scale=ScalarParameter() {defaultValue=0.6, group=Icon, parameterName=IconScale, sortPriority=3}) {materialFunction=&ScaleUVsByCenter}.UVs
   Tex=TextureObjectParameter() {group=Icon, parameterName=Icon, sortPriority=1}
  ) {
   samplerSource=SSM_Clamp_WorldGroupSettings
   samplerType=SAMPLERTYPE_Alpha
   texture=/FortUILibrary/Textures/Icons/ToBeDeprecated/T_UI_Icon_Global_Check
  }
  LinearInterpolate(
   A=LinearInterpolate(
    A=VectorParameter() {group=Circle, parameterName=CircleColor, defaultValue=<rgba 0 0.3512 0.7188 1>}.RGB
    B=VectorParameter() {group=Circle, parameterName=OutlineColor, defaultValue=<rgba 0.0052 0.5333 0 1>}.RGB
    Alpha=Max(A=@n1, B=@n2)
   )
   B=VectorParameter() {group=Icon, parameterName=IconColor, sortPriority=2, defaultValue=<rgba 1 0.8963 0 1>}.RGB
   Alpha=MaterialFunctionCall(DistanceField=@n3.R, Threshold=Constant() {r=0.5}) {materialFunction=&MF_UI_AntiAliasing}.Outer
  )
  Max(A=NamedRerouteUsage(), B=NamedRerouteUsage())
  n4 = MaterialFunctionCall!(Value PX=Multiply(A=ScalarParameter() {group=Layout, parameterName=EdgeOffset_PX}) {constB=2}, UV Pixel Size, X or Y) {materialFunction=/FortUILibrary/MaterialFunctions/Basics/MF_UI_PixelsToNormalized}
  n5 = MaterialFunctionCall(Value PX=ScalarParameter() {defaultValue=2, group=Circle, parameterName=Outline_PX}) {materialFunction=&MF_UI_PixelsToNormalized}
  n6 = MaterialFunctionCall(Size=OneMinus(None=Add(A=@n4.Normalized, B=@n5.Normalized) {constB=0}), OutlineThickness=@n5.Normalized) {materialFunction=&MF_UI_Circle}
  NamedRerouteDeclaration(None=@n6.Fill) {name=Circle, nodeColor=<rgba 0 0 0 1>}
  NamedRerouteDeclaration(None=@n6.Outline) {name=CircleOutline, nodeColor=<rgba 0 0 0 1>}
 ]
}
