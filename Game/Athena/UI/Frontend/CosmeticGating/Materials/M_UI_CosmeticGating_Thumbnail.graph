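# UI material for thumbnails in the frontend cosmetic gating screens, set up as translucent.
# A `Texture` parameter is sampled with normalized UI UVs, which MF_UI_UVScaling adjusts by the texture's size to scale the image to fill.
# The shared CoreUI rounded mask function supplies rounded corners with a fixed radius given in pixels.
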Material /Game/Athena/UI/Frontend/CosmeticGating/Materials/M_UI_CosmeticGating_Thumbnail
graph {
 properties {
  bUsedWithStaticMesh=false
  blendMode=BLEND_Translucent
  materialDomain=MD_UI
  pixelDepthOffsetMode=PDOM_Legacy
  usedShadingModels=Default Lit
 }
 nodes [
  MaterialFunctionCall(Corner Radius=Constant() {r=10}, Radius in Pixels=Constant() {r=1}) {materialFunction=&MF_UI_CoreUI_RoundedMask}
  n1 = TextureObjectParameter() {parameterName=Texture}
  TextureSample(
   UVs=MaterialFunctionCall(
    UV=MaterialFunctionCall() {materialFunction=&GetUserInterfaceUV}.Normalized UV
    TextureSize=TextureProperty(Tex=@n1)
   ) {materialFunction=&MF_UI_UVScaling}.Scaled To Fill
   Tex=@n1
  )
 ]
}
