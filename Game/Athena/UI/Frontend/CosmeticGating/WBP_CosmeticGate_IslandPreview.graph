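# Widget Blueprint for a frontend cosmetic-gating island preview, showing a thumbnail, approved/rejected status indicator, rating icon and rating name.
# Setter functions switch the check/close indicator, set rating text, show or collapse the rating icon, and feed textures into the thumbnail's dynamic material.
# LoadThumbnail async-loads a soft texture; UnloadThumbnail resets it and unloads the stored primary asset.
# PreConstruct previews defaults at design time, while Construct enlarges both images on mobile.
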WidgetBlueprint /Game/Athena/UI/Frontend/CosmeticGating/WBP_CosmeticGate_IslandPreview
graphs [
 EventGraph {nodes [
   n1 = LoadThumbnail!() -> (OutputDelegate, then, SoftTexture:Texture 2D Soft Object Reference) {is=title K2Node_CustomEvent}
   > n2 = Async Load Asset(Asset=@n1.SoftTexture)
   n3 = Cast To Texture2D(execute=@n2.Completed, Object=@n2)
   > n4 = GetDynamicMaterial(self=Get Image_Thumbnail() -> (Image_Thumbnail:Image Object Reference))
   > SetTextureParameterValue(self=@n4, ParameterName=Texture, Value=@n3)
   UnloadThumbnail() {is=title K2Node_CustomEvent}
   > n5 = GetDynamicMaterial(self=Get Image_Thumbnail())
   > SetTextureParameterValue(self=@n5, ParameterName=Texture, Value=/MeshModelingToolsetExp/Textures/DefaultOpacity)
   > UnloadPrimaryAsset!(
    execute
    PrimaryAssetId:Primary Asset Id Structure=GetPrimaryAssetIdFromSoftObjectReference!(SoftObjectReference:Soft Object Reference=Get SoftThumbnailImage() -> (SoftThumbnailImage:Texture 2D Soft Object Reference)) -> (ReturnValue:Primary Asset Id Structure)
   ) -> (then)
   n6 = Event PreConstruct()
   > Branch(Condition=@n6.IsDesignTime)
   > SetApprovalStatus!(execute, self, ApprovalStatus:bool=Get IsApproved() -> (IsApproved:bool)) -> (then)
   > LoadThumbnail!(execute, self, SoftTexture:Texture 2D Soft Object Reference=Get SoftThumbnailImage()) -> (then) {is=title K2Node_CallFunction}
   Event Construct()
   > SetMobileSize!(
    execute
    self
    Image_RatingIcon:Image Object Reference=Get Image_RatingIcon() -> (Image_RatingIcon:Image Object Reference)
    Image_Thumbnail:Image Object Reference=Get Image_Thumbnail()
   ) -> (then)
  ]}
 MobileScaleImage {nodes [
   n1 = Inputs!() -> (execute, Target Brush:Image Object Reference)
   n2 = Break Slate Brush!(SlateBrush:"Slate Brush Structure (by ref)"=Get Brush!(self:Image Object Reference=@n1) -> (Brush:Slate Brush Structure)) -> (
    TintColor:Slate Color Structure
    DrawAs:ESlateBrushDrawType Enum
    Tiling:ESlateBrushTileType Enum
    Mirroring:ESlateBrushMirrorType Enum
    ImageSize:Vector 2D Structure
    Margin:Margin Structure
    ResourceObject:Object Reference
    OutlineSettings:Slate Brush Outline Settings Structure
   ) {is=title K2Node_BreakStruct}
   n3 = Get MobileScaleFactor() -> (MobileScaleFactor:Vector 2D Structure)
   > Set Brush!(
    execute
    Brush:Slate Brush Structure=Make Slate Brush!(
     TintColor:Slate Color Structure=@n2
     DrawAs:ESlateBrushDrawType Enum=@n2.DrawAs
     Tiling:ESlateBrushTileType Enum=@n2.Tiling
     Mirroring:ESlateBrushMirrorType Enum=@n2.Mirroring
     ImageSize:Vector 2D Structure=vector2d * vector2d(A:Vector 2D Structure=@n2.ImageSize, B:Vector 2D Structure=@n3) -> (ReturnValue:Vector 2D Structure) {is=title K2Node_PromotableOperator}
     Margin:Margin Structure=@n2.Margin
     ResourceObject:Object Reference=@n2.ResourceObject
     OutlineSettings:Slate Brush Outline Settings Structure=@n2.OutlineSettings
    ) -> (SlateBrush:Slate Brush Structure) {is=title K2Node_MakeStruct}
    self:Image Object Reference=@n1
   ) -> (then, Output_Get:Slate Brush Structure)
   > Outputs!(then:Exec)
  ]}
 SetApprovalStatus {nodes [
   n1 = SetApprovalStatus!() -> (then, ApprovalStatus:bool=~) {is=title K2Node_FunctionEntry}
   > Set IsApproved(IsApproved:bool=@n1) -> (Output_Get:bool)
   n2 = Get WBP_UIKit_StatusIndicator() -> (WBP_UIKit_StatusIndicator:WBP UIKit Status Indicator Object Reference)
   > SetIndicatorType!(
    execute
    self:WBP UIKit Status Indicator Base Object Reference=@n2
    IndicatorType:E_UI_StatusIndicator_Type Enum=Select(
     Option 0:E_UI_StatusIndicator_Type Enum=NewEnumerator4
     Option 1:E_UI_StatusIndicator_Type Enum=NewEnumerator0
     Index=Get ApprovalStatus() -> (ApprovalStatus:bool)
    ) -> (ReturnValue:E_UI_StatusIndicator_Type Enum)
   ) -> (then)
   > SetIcon!(
    execute
    self:WBP UIKit Status Indicator Base Object Reference=@n2
    IconAsset:Object Reference=Select(
     Option 0:Object Reference=/FortUILibrary/Textures/MaterialInstances/MI_UIKit_Icon_Close
     Option 1:Object Reference=/FortUILibrary/Textures/MaterialInstances/MI_UIKit_Icon_Check
     Index=Get ApprovalStatus()
    ) -> (ReturnValue:Object Reference)
   ) -> (then)
  ]}
 SetMobileSize {nodes [
   n1 = SetMobileSize!() -> (then, Image_RatingIcon:Image Object Reference, Image_Thumbnail:Image Object Reference) {is=title K2Node_FunctionEntry}
   > Branch(Condition=IsMobileGame())
   > n2 = For Each Loop(Array:Array of Image Object References=Make Array!("[0]":Image Object Reference=@n1, "[1]":Image Object Reference=@n1.Image_Thumbnail) -> (Array:Array of Image Object References)) -> (Array Element:Image Object Reference)
   > Mobile Scale Image!(execute:Exec, Target Brush:Image Object Reference=@n2) -> (then:Exec) {is=title K2Node_MacroInstance}
  ]}
 SetRatingIcon {nodes [
   n1 = SetRatingIcon!() -> (then, RatingIcon:Texture 2DDynamic Object Reference) {is=title K2Node_FunctionEntry}
   > n2 = Is Valid(InputObject=@n1)
   > SetBrushFromTextureDynamic!(execute, self:Image Object Reference=Get Image_RatingIcon(), Texture:Texture 2DDynamic Object Reference=@n1, bMatchSize:bool=false) -> (then)
   > SetVisibility(self=Get Image_RatingIcon(), InVisibility=Visible)
   SetVisibility(execute=@n2.Is Not Valid, self=Get Image_RatingIcon(), InVisibility=Collapsed)
  ]}
 SetRatingText {nodes [
   n1 = SetRatingText!() -> (then, RatingText:Text) {is=title K2Node_FunctionEntry}
   > Set Text(Text=@n1, self=Get Text_RatingName() -> (Text_RatingName:Common Text Object Reference))
  ]}
 SetThumbnailImage {nodes [
   n1 = SetThumbnailImage!() -> (then, Thumbnail:Texture 2D Object Reference) {is=title K2Node_FunctionEntry}
   > n2 = GetDynamicMaterial(self=Get Image_Thumbnail())
   > SetTextureParameterValue(self=@n2, ParameterName=Texture, Value=@n1)
   n3 = Get Image_LazyThumbnail() -> (Image_LazyThumbnail:Fort Lazy Image Object Reference)
   GetDynamicMaterial(self=@n3)
   SetBrushFromLazyMaterial!(execute, self:Common Lazy Image Object Reference=@n3, LazyMaterial:"Material Interface Soft Object Reference (by ref)") -> (then)
   Get SoftThumbnailImage()
  ]}
]
variables [
 IsApproved {type=boolean, default=true}
 MobileScaleFactor {type=Vector2D, default=<2 2>}
 SoftThumbnailImage {type=Texture2D, default=/Game/Athena/UI/Frontend/CosmeticGating/Textures/T_UI_Thumbnail_Juno}
]
widgets {
 info {inheritedWidgetCount=0, parentClass=&UserWidget, rootWidgetClass=&GridPanel, widgetCount=5}
 widgets [
  GridPanel_0 {widgetClassPath=&GridPanel, widget=GridPanel_0}
  WBP_UIKit_StatusIndicator {
   widgetClassPath=/FortUILibrary/UIKit/StatusIndicator/WBP_UIKit_StatusIndicator
   bIsVariable=true
   parent=GridPanel_0
   widget=WBP_UIKit_StatusIndicator
   properties {
    icon=&MI_UIKit_Icon_Check
    overrideStatusIndicatorState=true
    overrideStatusIndicatorVisibility=true
   }
   slot {class=GridSlot, properties {rowSpan=1}}
  }
  Image_Thumbnail {
   widgetClassPath=&Image
   bIsVariable=true
   parent=GridPanel_0
   widget=Image_Thumbnail
   properties {
    visibility=SelfHitTestInvisible
    brush {
     bIsDynamicallyLoaded=false
     drawAs=Image
     imageType=FullColor
     mirroring=NoMirror
     resourceName=None
     resourceObject=/Game/Athena/UI/Frontend/CosmeticGating/Materials/M_UI_CosmeticGating_Thumbnail
     tiling=NoTile
     imageSize=<151 85>
     margin=<ltrb 0 0 0 0>
     outlineSettings {
      bUseBrushTransparency=false
      roundingType=HalfHeightRadius
      width=0
      color {colorUseRule=UseColor_Specified, specifiedColor=<rgba 0 0 0 0>}
      cornerRadii=<0 0 0 0>
     }
     tintColor {colorUseRule=UseColor_Specified, specifiedColor=<rgba 1 1 1 1>}
     uVRegion {bIsValid=false, max=<0 0>, min=<0 0>}
    }
    renderTransform {angle=0, scale=<1 1>, shear=<0 0>, translation=<0 -6>}
   }
   slot {class=GridSlot, properties {horizontalAlignment=HAlign_Center, row=1, rowSpan=1}}
  }
  Image_RatingIcon {
   widgetClassPath=&Image
   bIsVariable=true
   parent=GridPanel_0
   widget=Image_RatingIcon
   properties {
    visibility=Collapsed
    brush {
     bIsDynamicallyLoaded=false
     drawAs=Image
     imageType=NoImage
     mirroring=NoMirror
     resourceName=None
     resourceObject=null
     tiling=NoTile
     imageSize=<40 40>
     margin=<ltrb 0 0 0 0>
     outlineSettings {
      bUseBrushTransparency=false
      roundingType=HalfHeightRadius
      width=0
      color {colorUseRule=UseColor_Specified, specifiedColor=<rgba 0 0 0 0>}
      cornerRadii=<0 0 0 0>
     }
     tintColor {colorUseRule=UseColor_Specified, specifiedColor=<rgba 1 1 1 1>}
     uVRegion {bIsValid=false, max=<0 0>, min=<0 0>}
    }
    renderTransform {angle=0, scale=<1 1>, shear=<0 0>, translation=<0 -6>}
   }
   slot {class=GridSlot, properties {horizontalAlignment=HAlign_Left, row=1, rowSpan=1, verticalAlignment=VAlign_Bottom, nudge=<4 -4>}}
  }
  Text_RatingName {
   widgetClassPath=&CommonTextBlock
   bIsVariable=true
   parent=GridPanel_0
   widget=Text_RatingName
   properties {
    mobileFontSizeMultiplier=1.2
    text="[PH] E10"
    font {
     bForceMonospaced=false
     bMaterialIsStencil=false
     fontMaterial=null
     fontObject=&HeadingNow
     letterSpacing=0
     monospacedWidth=1
     size=15
     skewAmount=0
     typefaceFontName=74 Regular
     outlineSettings {
      bApplyOutlineToDropShadows=false
      bMiteredCorners=false
      bSeparateFillAlpha=false
      outlineMaterial=null
      outlineSize=0
      outlineColor=<rgba 0 0 0 1>
     }
    }
    shadowOffset=<0 0>
   }
   slot {class=GridSlot, properties {horizontalAlignment=HAlign_Center, row=2, rowSpan=1}}
  }
 ]
}
