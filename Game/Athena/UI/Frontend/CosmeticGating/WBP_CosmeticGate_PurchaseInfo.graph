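# Shop purchase info widget that explains which age-rated experiences allow an offer's restricted cosmetics or vocal Jam Tracks.
# VM_OfferUpdated extracts the offer's gating interfaces, then sets warning and body text and shows the purchase question based on eligibility or vocal restriction.
# Age-rating data fills two or three island previews, labelling the last "or above" when ratings outnumber previews.
# Thumbnails follow the player's ratings board, and PreConstruct applies configurable fonts, visibility and alignment.
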WidgetBlueprint /Game/Athena/UI/Frontend/CosmeticGating/WBP_CosmeticGate_PurchaseInfo
graphs [
 EventGraph {nodes [
   LoadThumbnails() {is=title K2Node_CustomEvent}
   > SetIslandThumbnails()
   UnloadThumbnails() {is=title K2Node_CustomEvent}
   > UnloadThumbnail(self:WBP Cosmetic Gate Island Preview Object Reference=[
     Get IslandPreview_E() -> (IslandPreview_E:WBP Cosmetic Gate Island Preview Object Reference)
     Get IslandPreview_E10() -> (IslandPreview_E10:WBP Cosmetic Gate Island Preview Object Reference)
     Get IslandPreview_T() -> (IslandPreview_T:WBP Cosmetic Gate Island Preview Object Reference)
    ])
   Event PreConstruct()
   > SetVisuals()
  ]}
 GetIslandPreviewsFromRatingInfo {nodes [
   n1 = GetIslandPreviewsFromRatingInfo!() -> (then, RatingIcons:Array of Age Rating Info Structures) {is=title K2Node_FunctionEntry}
   n2 = Get IslandPreview_T()
   n3 = Get IslandPreview_E10()
   n4 = Get IslandPreview_E()
   n5 = Get IslandPreview_E10()
   n6 = Get IslandPreview_T()
   n7 = Length(TargetArray:Array of Age Rating Info Structures=@n1)
   > Return Node!(
    execute
    IslandPreviews:Array of WBP Cosmetic Gate Island Preview Object References=Select(
     Option 0:Array of WBP Cosmetic Gate Island Preview Object References=Make Array!(
      "[0]":WBP Cosmetic Gate Island Preview Object Reference=@n4
      "[1]":WBP Cosmetic Gate Island Preview Object Reference=@n5
      "[2]":WBP Cosmetic Gate Island Preview Object Reference=@n6
     ) -> (Array:Array of WBP Cosmetic Gate Island Preview Object References)
     Option 1:Array of WBP Cosmetic Gate Island Preview Object References=Make Array!("[0]"=@n3, "[1]"=@n2) -> (Array)
     Index="integer <= integer"(A=@n7, B=2)
    ) -> (ReturnValue:Array of WBP Cosmetic Gate Island Preview Object References)
   )
  ]}
 SetIslandAgeRestrictionInfo {nodes [
   SetIslandAgeRestrictionInfo!() -> (then, CosmeticItemInterfaces:Array of Cosmetic Item Gating Interface Interfaces) {is=title K2Node_FunctionEntry}
   > SetVisibility(self=[Get IslandPreview_E(), Get IslandPreview_E10(), Get IslandPreview_T()], InVisibility=Collapsed)
   > n1 = GetAgeRatingInfosForExperiences!(
    execute
    LocalPlayer:Local Player Object Reference=GetOwningLocalPlayer!(self:Widget Object Reference) -> (ReturnValue:Local Player Object Reference)
    CosmeticItems:Array of Cosmetic Item Gating Interface Interfaces=Get CosmeticItemInterfaces() -> (CosmeticItemInterfaces:Array of Cosmetic Item Gating Interface Interfaces)
   ) -> (then, ReturnValue:Array of Age Rating Info Structures)
   > n2 = Set RatingInfoArray(RatingInfoArray:Array of Age Rating Info Structures=@n1) -> (Output_Get:Array of Age Rating Info Structures)
   > n3 = Set PreviewArray(PreviewArray:Array of WBP Cosmetic Gate Island Preview Object References=GetIslandPreviewsFromRatingInfo!(RatingIcons:Array of Age Rating Info Structures=@n2) -> (IslandPreviews:Array of WBP Cosmetic Gate Island Preview Object References) {is=title K2Node_CallFunction}) -> (Output_Get:Array of WBP Cosmetic Gate Island Preview Object References)
   > n4 = For Each Loop(Array:Array of WBP Cosmetic Gate Island Preview Object References=@n3) -> (Array Element:WBP Cosmetic Gate Island Preview Object Reference)
   > Set LoopPreview(LoopPreview:WBP Cosmetic Gate Island Preview Object Reference=@n4) -> (Output_Get:WBP Cosmetic Gate Island Preview Object Reference)
   > Set LoopIndex(LoopIndex:int=@n4.Array Index) -> (Output_Get:int=0)
   > Branch(Condition=Is Valid Index!(
     TargetArray:Array of Age Rating Info Structures=Get RatingInfoArray() -> (RatingInfoArray:Array of Age Rating Info Structures)
     IndexToTest:int=Get LoopIndex() -> (LoopIndex:int=0)
    ) -> (ReturnValue:bool))
   n5 = "Get (a copy)"!(Array:Array of Age Rating Info Structures=Get RatingInfoArray(), Dimension 1=Get LoopIndex() -> (LoopIndex=0)) -> (Output_AgeRatingText:Text, Output_RatingImageTexture:Texture 2DDynamic Object Reference, Output_bApprovalStatus:bool=~)
   > SetRatingIcon!(
    execute
    self:WBP Cosmetic Gate Island Preview Object Reference=Get LoopPreview() -> (LoopPreview:WBP Cosmetic Gate Island Preview Object Reference)
    RatingIcon:Texture 2DDynamic Object Reference=@n5.Output_RatingImageTexture
   ) -> (then)
   n6 = Get LoopIndex() -> (LoopIndex=0)
   n7 = "Equal (Integer)"(
    A=Last Index!(TargetArray:Array of WBP Cosmetic Gate Island Preview Object References=Get PreviewArray() -> (PreviewArray:Array of WBP Cosmetic Gate Island Preview Object References)) -> (ReturnValue=0)
    B=@n6
   )
   n8 = "integer < integer"(A=@n6, B=Last Index(TargetArray:Array of Age Rating Info Structures=Get RatingInfoArray()) -> (ReturnValue=0))
   > SetRatingText!(
    execute
    self:WBP Cosmetic Gate Island Preview Object Reference=Get LoopPreview()
    RatingText:Text=Select(Option 0:Text=@n5, Option 1:Text=Format Text!(Format="LOCTEXT(\"{0} or above\")", 0:Text=@n5) -> (Result), Index=AND Boolean(A=@n7, B=@n8)) -> (ReturnValue:Text)
   ) -> (then)
   > SetApprovalStatus!(
    execute
    self:WBP Cosmetic Gate Island Preview Object Reference=Get LoopPreview()
    ApprovalStatus:bool="Get (a copy)"(Array=Get RatingInfoArray(), Dimension 1=Get LoopIndex() -> (LoopIndex=0)) -> (Output_bApprovalStatus=~).Output_bApprovalStatus
   ) -> (then)
   > SetVisibility(self=Get LoopPreview(), InVisibility=SelfHitTestInvisible)
  ]}
 SetIslandThumbnails {nodes [
   SetIslandThumbnails() {is=title K2Node_FunctionEntry}
   > n1 = Switch on ERatingsBoard!(
    execute
    Selection:ERatingsBoard Enum=GetUserRatingBoard!(LocalPlayer:Local Player Object Reference=GetOwningLocalPlayer()) -> (ReturnValue:ERatingsBoard Enum=ESRB)
   ) -> (
    ESRB:Exec
    PEGI:Exec
    USK:Exec
    GRAC:Exec
    IARC:Exec
    ACB:Exec
    ClassInd:Exec
    Russia:Exec
    CERO:Exec
    FPB:Exec
    GCAM:Exec
    MRO:Exec
    DGSC:Exec
    GMedia:Exec
    IGRS:Exec
    None:Exec
   ) {is=title K2Node_SwitchEnum}
   LoadThumbnail!(
    execute=[@n1.ACB, @n1.GRAC]
    self:WBP Cosmetic Gate Island Preview Object Reference=Get IslandPreview_E()
    SoftTexture:Texture 2D Soft Object Reference=/Game/Athena/UI/Frontend/CosmeticGating/Textures/T_UI_Thumbnail_PropHunt
   ) -> (then)
   > LoadThumbnail(self=Get IslandPreview_E10(), SoftTexture=/Game/Athena/UI/Frontend/CosmeticGating/Textures/T_UI_Thumbnail_Sparks)
   > LoadThumbnail(self=Get IslandPreview_T(), SoftTexture=/Game/Athena/UI/Frontend/CosmeticGating/Textures/T_UI_Thumbnail_Athena)
   LoadThumbnail(
    execute=[@n1.CERO, @n1.ClassInd, @n1.ESRB, @n1.FPB, @n1.GCAM, @n1.IARC, @n1.MRO, @n1.PEGI, @n1.Russia, @n1.USK]
    self=Get IslandPreview_E()
    SoftTexture=&T_UI_Thumbnail_PropHunt
   )
   > LoadThumbnail(self=Get IslandPreview_E10(), SoftTexture=/Game/Athena/UI/Frontend/CosmeticGating/Textures/T_UI_Thumbnail_Juno)
   > LoadThumbnail(self=Get IslandPreview_T(), SoftTexture=&T_UI_Thumbnail_Sparks)
  ]}
 SetVisuals {nodes [
   SetVisuals() {is=title K2Node_FunctionEntry}
   > n1 = Sequence!(execute) -> (then_0, then_1, then_2:Exec, then_3:Exec)
   > Set Visibility(
    Visibility=Select(Option 0:ESlateVisibility Enum=Collapsed, Option 1:ESlateVisibility Enum=SelfHitTestInvisible, Index=Get Show Header() -> (Show Header:bool)) -> (ReturnValue:ESlateVisibility Enum)
    self=Get HBox_Header() -> (HBox_Header:Horizontal Box Object Reference)
   )
   > Set Visibility(
    Visibility=Select(Option 0=Collapsed, Option 1=SelfHitTestInvisible, Index=Get Show Header Icon() -> (Show Header Icon:bool))
    self=Get Icon_Header() -> (Icon_Header:Image Object Reference)
   )
   Set Font!(
    execute=@n1.then_2
    Font:Slate Font Info Structure=Get Header Font() -> (Header Font:Slate Font Info Structure)
    self:Text Object Reference=Get Text_Header() -> (Text_Header:Common Text Object Reference)
   ) -> (then, Output_Get:Slate Font Info Structure)
   > Set Font(
    Font=Get Sub Header Font() -> (Sub Header Font:Slate Font Info Structure)
    self=Get Text_SubHeader() -> (Text_SubHeader:Common Text Object Reference)
   )
   > n2 = Set Font(Font=Get Body Text Font() -> (Body Text Font:Slate Font Info Structure), self=Get Text_Body() -> (Text_Body:Common Text Object Reference))
   > Set Font(Font=@n2, self=Get Text_PurchaseConfirmation() -> (Text_PurchaseConfirmation:Common Text Object Reference))
   n3 = SlotAsVerticalBoxSlot(Widget=Get HBox_IslandPreviews() -> (HBox_IslandPreviews:Horizontal Box Object Reference))
   n4 = Set HorizontalAlignment!(
    execute=@n1.then_3
    HorizontalAlignment:EHorizontalAlignment Enum=Get Usability Matrix Alignment() -> (Usability Matrix Alignment:EHorizontalAlignment Enum=HAlign_Fill)
    self:Vertical Box Slot Object Reference=@n3
   ) -> (then, Output_Get:EHorizontalAlignment Enum=HAlign_Fill)
   n5 = Get!(self:Vertical Box Slot Object Reference=@n3) -> (Padding_Left:float=~, Padding_Top:float=~, Padding_Right:float=~, Padding_Bottom:float=~)
   > Set!(
    execute
    Padding_Left:float=Select!(HAlign_Fill:float=0, HAlign_Left:float=-14, HAlign_Center:float=0, HAlign_Right:float=-14, Index:EHorizontalAlignment Enum=@n4) -> (ReturnValue:float)
    Padding_Top:float=@n5.Padding_Top
    Padding_Right:float=@n5.Padding_Right
    Padding_Bottom:float=@n5.Padding_Bottom
    self:Vertical Box Slot Object Reference=@n3
   ) -> (then, Output_Get:Margin Structure)
   Set Visibility(
    execute=@n1.then_1
    Visibility=Select!(
     Option 0:ESlateVisibility Enum=Collapsed
     Option 1:ESlateVisibility Enum=SelfHitTestInvisible
     Index:bool=Get Show Sub Header Icon() -> (Show Sub Header Icon:bool)
    ) -> (ReturnValue:ESlateVisibility Enum)
    self=Get Icon_Warning() -> (Icon_Warning:Image Object Reference)
   )
  ]}
 VM_OfferUpdated {nodes [
   n1 = VM_OfferUpdated!() -> (then, OfferVM:Fort Item Shop Offer VM Object Reference) {is=title K2Node_FunctionEntry}
   > n2 = Get Content Gating Interfaces From Offer!(execute, OfferVM:Fort Item Shop Offer VM Object Reference=@n1) -> (then, Interfaces:Array of Cosmetic Item Gating Interface Interfaces) {is=title K2Node_CallFunction}
   > Set CosmeticGatingInterfaces(CosmeticGatingInterfaces:Array of Cosmetic Item Gating Interface Interfaces=@n2) -> (Output_Get:Array of Cosmetic Item Gating Interface Interfaces)
   > n3 = Sequence()
   > n4 = IsPlayerEligibleToUseRestrictedItems!(
    execute
    LocalPlayer:Local Player Object Reference=GetOwningLocalPlayer()
    CosmeticItems:Array of Cosmetic Item Gating Interface Interfaces=Get CosmeticGatingInterfaces() -> (CosmeticGatingInterfaces:Array of Cosmetic Item Gating Interface Interfaces)
   ) -> (then, ReturnValue:bool)
   > Set IsPlayerEligibleToUseRestrictedItems(IsPlayerEligibleToUseRestrictedItems:bool=@n4) -> (Output_Get:bool)
   n5 = ItemsHaveVocalRatingRestriction!(
    execute=@n3.then_1
    LocalPlayer:Local Player Object Reference=GetOwningLocalPlayer()
    CosmeticItems:Array of Cosmetic Item Gating Interface Interfaces=Get CosmeticGatingInterfaces()
   ) -> (then, ReturnValue:bool)
   > Set OfferHasVocalRestriction(OfferHasVocalRestriction:bool=@n5) -> (Output_Get:bool)
   SetIslandAgeRestrictionInfo!(execute=@n3.then_3, self, CosmeticItemInterfaces:Array of Cosmetic Item Gating Interface Interfaces=Get CosmeticGatingInterfaces()) -> (then) {is=title K2Node_CallFunction}
   > SetIslandThumbnails() {is=title K2Node_CallFunction}
   n6 = Get OfferVM() -> (OfferVM:Fort Item Shop Offer VM Object Reference)
   n7 = "Equal (Integer)"(
    A=Length(TargetArray:Array of Fort Item VM Object References=Get GrantedItems!(self:Fort Item Shop Offer VM Object Reference=@n6) -> (GrantedItems:Array of Fort Item VM Object References))
    B=1
   )
   Set Text(
    execute=@n3.then_2
    Text=Select(
     Option 0:Text="LOCTEXT(\"Some of these items can\\'t be worn in all experiences!\")"
     Option 1:Text=Select(
      Option 0="LOCTEXT(\"Some of these items can\\'t be used in all experiences!\")"
      Option 1="LOCTEXT(\"This Jam Track contains lyrics not suitable for all audiences.\")"
      Index=@n7
     )
     Index=Get OfferHasVocalRestriction() -> (OfferHasVocalRestriction:bool)
    ) -> (ReturnValue:Text)
    self=Get Text_SubHeader()
   )
   > n8 = GetPurchaseConfirmationRatingRestrictionText!(
    execute
    LocalPlayer:Local Player Object Reference=GetOwningLocalPlayer()
    OfferGrantedCosmetics:Array of Cosmetic Item Gating Interface Interfaces=Get CosmeticGatingInterfaces()
   ) -> (then, ReturnValue:Text)
   > Set Text(Text=@n8, self=Get Text_Body())
   n9 = Get OfferHasVocalRestriction()
   n10 = Get IsPlayerEligibleToUseRestrictedItems() -> (IsPlayerEligibleToUseRestrictedItems:bool)
   > Set Visibility(
    Visibility=Select(Option 0:ESlateVisibility Enum=Collapsed, Option 1:ESlateVisibility Enum=SelfHitTestInvisible, Index=OR Boolean(A=@n9, B=@n10)) -> (ReturnValue:ESlateVisibility Enum)
    self=Get Text_PurchaseConfirmation()
   )
  ]}
]
variables [
 Body Text Font {
  type=SlateFontInfo
  default {
   bForceMonospaced=false
   bMaterialIsStencil=false
   fontMaterial=null
   fontObject=&HeadingNow
   letterSpacing=0
   monospacedWidth=1
   size=15.75
   skewAmount=0
   typefaceFontName=74 Regular
   outlineSettings {
    bApplyOutlineToDropShadows=false
    bMiteredCorners=false
    bSeparateFillAlpha=false
    outlineMaterial=null
    outlineSize=0
    outlineColor=<rgba 0 0 0 1>
   }
  }
 }
 Header Font {
  type=SlateFontInfo
  default {
   bForceMonospaced=false
   bMaterialIsStencil=false
   fontMaterial=null
   fontObject=&HeadingNow
   letterSpacing=0
   monospacedWidth=1
   size=20.25
   skewAmount=0
   typefaceFontName=75 Medium
   outlineSettings {
    bApplyOutlineToDropShadows=false
    bMiteredCorners=false
    bSeparateFillAlpha=false
    outlineMaterial=null
    outlineSize=0
    outlineColor=<rgba 0 0 0 1>
   }
  }
 }
 Show Header {type=boolean, default=true}
 Show Header Icon {type=boolean, default=true}
 Show Sub Header Icon {type=boolean, default=true}
 Sub Header Font {
  type=SlateFontInfo
  default {
   bForceMonospaced=false
   bMaterialIsStencil=false
   fontMaterial=null
   fontObject=&HeadingNow
   letterSpacing=0
   monospacedWidth=1
   size=18
   skewAmount=0
   typefaceFontName=75 Medium
   outlineSettings {
    bApplyOutlineToDropShadows=false
    bMiteredCorners=false
    bSeparateFillAlpha=false
    outlineMaterial=null
    outlineSize=0
    outlineColor=<rgba 0 0 0 1>
   }
  }
 }
 Usability Matrix Alignment {type=EHorizontalAlignment, default=HAlign_Left}
]
widgets {
 info {inheritedWidgetCount=0, parentClass=&UserWidget, rootWidgetClass=&VerticalBox, widgetCount=13}
 widgets [
  VB_All {widgetClassPath=&VerticalBox, widget=VB_All}
  HBox_Header {
   widgetClassPath=&HorizontalBox
   bIsVariable=true
   parent=VB_All
   widget=HBox_Header
   properties {flowDirectionPreference=Culture}
   slot {class=VerticalBoxSlot, properties {padding=<ltrb 0 0 0 16>}}
  }
  Icon_Header {
   widgetClassPath=&Image
   bIsVariable=true
   parent=HBox_Header
   widget=Icon_Header
   properties {brush {
     bIsDynamicallyLoaded=false
     drawAs=Image
     imageType=NoImage
     mirroring=NoMirror
     resourceName=None
     resourceObject=/FortUILibrary/Textures/MaterialInstances/MI_UIKit_Icon_Information
     tiling=NoTile
     imageSize=<30 30>
     margin=<ltrb 0 0 0 0>
     outlineSettings {
      bUseBrushTransparency=false
      roundingType=HalfHeightRadius
      width=0
      color {colorUseRule=UseColor_Specified, specifiedColor=<rgba 0 0 0 0>}
      cornerRadii=<0 0 0 0>
     }
     tintColor {colorUseRule=UseColor_Specified, specifiedColor=<rgba 1 1 1 1>}
     uVRegion {bIsValid=false, max=<0 0>, min=<0 0>}
    }}
   slot {class=HorizontalBoxSlot, properties {horizontalAlignment=HAlign_Left, verticalAlignment=VAlign_Center, padding=<ltrb 0 0 8 0>}}
  }
  Text_Header {
   widgetClassPath=&CommonTextBlock
   bIsVariable=true
   parent=HBox_Header
   widget=Text_Header
   properties {
    text=Why do I need to unlock?
    font {
     bForceMonospaced=false
     bMaterialIsStencil=false
     fontMaterial=null
     fontObject=&BurbankSmall
     letterSpacing=0
     monospacedWidth=1
     size=18
     skewAmount=0
     typefaceFontName=Bold
     outlineSettings {
      bApplyOutlineToDropShadows=false
      bMiteredCorners=false
      bSeparateFillAlpha=false
      outlineMaterial=null
      outlineSize=0
      outlineColor=<rgba 0 0 0 1>
     }
    }
    shadowOffset=<0 0>
   }
   slot {class=HorizontalBoxSlot, properties {verticalAlignment=VAlign_Center, size {sizeRule=Fill, value=1}}}
  }
  HorizontalBox_1 {
   widgetClassPath=&HorizontalBox
   parent=VB_All
   widget=HorizontalBox_1
   properties {flowDirectionPreference=Culture}
   slot {class=VerticalBoxSlot, properties {padding=<ltrb 0 0 0 24>}}
  }
  Icon_Warning {
   widgetClassPath=&Image
   bIsVariable=true
   parent=HorizontalBox_1
   widget=Icon_Warning
   properties {brush {
     bIsDynamicallyLoaded=false
     drawAs=Image
     imageType=NoImage
     mirroring=NoMirror
     resourceName=None
     resourceObject=/Game/Athena/UI/Frontend/CosmeticGating/Materials/MI_UI_CosmeticGate_WarningIcon
     tiling=NoTile
     imageSize=<30 30>
     margin=<ltrb 0 0 0 0>
     outlineSettings {
      bUseBrushTransparency=false
      roundingType=HalfHeightRadius
      width=0
      color {colorUseRule=UseColor_Specified, specifiedColor=<rgba 0 0 0 0>}
      cornerRadii=<0 0 0 0>
     }
     tintColor {colorUseRule=UseColor_Specified, specifiedColor=<rgba 1 1 1 1>}
     uVRegion {bIsValid=false, max=<0 0>, min=<0 0>}
    }}
   slot {class=HorizontalBoxSlot, properties {horizontalAlignment=HAlign_Left, verticalAlignment=VAlign_Center, padding=<ltrb 0 0 8 0>}}
  }
  Text_SubHeader {
   widgetClassPath=&CommonTextBlock
   bIsVariable=true
   parent=HorizontalBox_1
   widget=Text_SubHeader
   properties {
    autoWrapText=true
    text="[PH] Some of these items can't be worn in all experiences!"
    font {
     bForceMonospaced=false
     bMaterialIsStencil=false
     fontMaterial=null
     fontObject=&BurbankSmall
     letterSpacing=0
     monospacedWidth=1
     size=18
     skewAmount=0
     typefaceFontName=Bold
     outlineSettings {
      bApplyOutlineToDropShadows=false
      bMiteredCorners=false
      bSeparateFillAlpha=false
      outlineMaterial=null
      outlineSize=0
      outlineColor=<rgba 0 0 0 1>
     }
    }
    shadowOffset=<0 0>
   }
   slot {class=HorizontalBoxSlot, properties {verticalAlignment=VAlign_Center, size {sizeRule=Fill, value=1}}}
  }
  HBox_IslandPreviews {
   widgetClassPath=&HorizontalBox
   bIsVariable=true
   parent=VB_All
   widget=HBox_IslandPreviews
   properties {flowDirectionPreference=Culture}
   slot {class=VerticalBoxSlot, properties {horizontalAlignment=HAlign_Left, padding=<ltrb -14 0 0 0>}}
  }
  IslandPreview_E {
   widgetClassPath=&/WBP_CosmeticGate_IslandPreview
   bIsVariable=true
   parent=HBox_IslandPreviews
   widget=IslandPreview_E
   properties {isApproved=false}
   slot {class=HorizontalBoxSlot, properties {padding=<ltrb 14 0 14 0>}}
  }
  IslandPreview_E10 {
   widgetClassPath=&WBP_CosmeticGate_IslandPreview
   bIsVariable=true
   parent=HBox_IslandPreviews
   widget=IslandPreview_E10
   properties {isApproved=false}
   slot {class=HorizontalBoxSlot, properties {padding=<ltrb 14 0 14 0>}}
  }
  IslandPreview_T {
   widgetClassPath=&WBP_CosmeticGate_IslandPreview
   bIsVariable=true
   parent=HBox_IslandPreviews
   widget=IslandPreview_T
   slot {class=HorizontalBoxSlot, properties {padding=<ltrb 14 0 14 0>}}
  }
  Text_Body {
   widgetClassPath=&CommonTextBlock
   bIsVariable=true
   parent=VB_All
   widget=Text_Body
   properties {
    autoWrapText=true
    text="[PH] You will ONLY be able to wear these items in experiences rated TEEN."
    font {
     bForceMonospaced=false
     bMaterialIsStencil=false
     fontMaterial=null
     fontObject=&BurbankSmall
     letterSpacing=0
     monospacedWidth=1
     size=18
     skewAmount=0
     typefaceFontName=Bold
     outlineSettings {
      bApplyOutlineToDropShadows=false
      bMiteredCorners=false
      bSeparateFillAlpha=false
      outlineMaterial=null
      outlineSize=0
      outlineColor=<rgba 0 0 0 1>
     }
    }
    shadowOffset=<0 0>
   }
   slot {class=VerticalBoxSlot, properties {padding=<ltrb 0 48 0 0>}}
  }
  Text_PurchaseConfirmation {
   widgetClassPath=&CommonTextBlock
   bIsVariable=true
   parent=VB_All
   widget=Text_PurchaseConfirmation
   properties {
    autoWrapText=true
    text=Would you like to proceed with your purchase?
    font {
     bForceMonospaced=false
     bMaterialIsStencil=false
     fontMaterial=null
     fontObject=&HeadingNow
     letterSpacing=0
     monospacedWidth=1
     size=15.75
     skewAmount=0
     typefaceFontName=74 Regular
     outlineSettings {
      bApplyOutlineToDropShadows=false
      bMiteredCorners=false
      bSeparateFillAlpha=false
      outlineMaterial=null
      outlineSize=0
      outlineColor=<rgba 0 0 0 1>
     }
    }
    shadowOffset=<0 0>
   }
   slot {class=VerticalBoxSlot, properties {padding=<ltrb 0 24 0 0>}}
  }
 ]
}
