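# Cosmetic gating warning icon widget for the frontend UI, a single Image showing the cosmetic gate warning material.
# PreConstruct passes the IconSize variable to SetIconSize, so each placed instance can be sized in the designer.
# SetIconSize stores the given size and rebuilds the image brush with it, copying the current brush's other properties such as tint, material and margin.
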WidgetBlueprint /Game/Athena/UI/Frontend/CosmeticGating/WBP_CosmeticGating_WarningIcon
graphs [
 EventGraph {nodes [
   Event PreConstruct()
   > SetIconSize!(execute, self, In Brush Image Size:Vector 2D Structure=Get IconSize() -> (IconSize:Vector 2D Structure)) -> (then)
  ]}
 SetIconSize {nodes [
   n1 = SetIconSize!() -> (then, In Brush Image Size:Vector 2D Structure) {is=title K2Node_FunctionEntry}
   > n2 = Set IconSize(IconSize:Vector 2D Structure=@n1) -> (Output_Get:Vector 2D Structure)
   n3 = Get Icon() -> (Icon:Image Object Reference)
   n4 = Get!(self:Image Object Reference=@n3) -> (
    Brush_TintColor:Slate Color Structure
    Brush_DrawAs:ESlateBrushDrawType Enum
    Brush_Tiling:ESlateBrushTileType Enum
    Brush_Mirroring:ESlateBrushMirrorType Enum
    Brush_ImageSize:Vector 2D Structure
    Brush_Margin:Margin Structure
    Brush_ResourceObject:Object Reference
    Brush_OutlineSettings:Slate Brush Outline Settings Structure
   )
   > SetBrush!(
    execute
    self=@n3
    InBrush_TintColor:Slate Color Structure=@n4
    InBrush_DrawAs:ESlateBrushDrawType Enum=@n4.Brush_DrawAs
    InBrush_Tiling:ESlateBrushTileType Enum=@n4.Brush_Tiling
    InBrush_Mirroring:ESlateBrushMirrorType Enum=@n4.Brush_Mirroring
    InBrush_ImageSize:Vector 2D Structure=@n2
    InBrush_Margin:Margin Structure=@n4.Brush_Margin
    InBrush_ResourceObject:Object Reference=@n4.Brush_ResourceObject
    InBrush_OutlineSettings:Slate Brush Outline Settings Structure=@n4.Brush_OutlineSettings
   ) -> (then)
  ]}
]
variables [IconSize {type=Vector2D, default=<40 40>}]
widgets {
 info {inheritedWidgetCount=0, parentClass=&UserWidget, rootWidgetClass=&Image, widgetCount=1}
 widgets [Icon {
   widgetClassPath=&Image
   bIsVariable=true
   widget=Icon
   properties {
    visibility=SelfHitTestInvisible
    brush {
     bIsDynamicallyLoaded=false
     drawAs=Image
     imageType=NoImage
     mirroring=NoMirror
     resourceName=None
     resourceObject=/Game/Athena/UI/Frontend/CosmeticGating/Materials/MI_UI_CosmeticGate_WarningIcon
     tiling=NoTile
     imageSize=<32 32>
     margin=<ltrb 0 0 0 0>
     outlineSettings {
      bUseBrushTransparency=false
      roundingType=HalfHeightRadius
      width=0
      color {colorUseRule=UseColor_Specified, specifiedColor=<rgba 0 0 0 0>}
      cornerRadii=<0 0 0 0>
     }
     tintColor {colorUseRule=UseColor_Specified, specifiedColor=<rgba 1 1 1 1>}
     uVRegion {bIsValid=false, max=<0 0>, min=<0 0>}
    }
   }
  }]
}
